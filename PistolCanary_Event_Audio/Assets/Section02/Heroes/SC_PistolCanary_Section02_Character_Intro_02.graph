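# Sound cue for the second Section 02 hero character intro in the PistolCanary event audio.
# A fixed short delay feeds a single wave player for the matching intro wav.
# It routes through the Fort_Events sound class with a slightly boosted volume and a very large max distance.
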SoundCue /PistolCanary_Event_Audio/Assets/Section02/Heroes/SC_PistolCanary_Section02_Character_Intro_02
root=Delay_1
nodes [
 {id=Delay_1, children [WavePlayer_1], properties {delayMax=0.12, delayMin=0.12}}
 {id=WavePlayer_1, properties {soundWaveAssetPtr=/PistolCanary_Event_Audio/Assets/Section02/Heroes/wav/PistolCanary_Character_Intro_02}}
]
properties {maxDistance=2097152, soundClassObject=&Fort_Events, volumeMultiplier=1.1}
