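# Empty SoundCue for hero screen side swipes in Section 02 of the PistolCanary event audio.
# It contains no sound nodes and has no root, so it has nothing to play.
# It carries only playback setup: a large point-of-interest attenuation asset, an extended max distance and the Fort_Events sound class.
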SoundCue /PistolCanary_Event_Audio/Assets/Section02/Heroes/SC_PistolCanary_Section02_HeroScreenSideSwipes
root=null
nodes []
properties {
 attenuationSettings=/PistolCanary_Audio/Assets/Settings/ATT_PistolCanary_POI_Large
 maxDistance=40000
 soundClassObject=&Fort_Events
}
