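# Sound Cue for a PistolCanary event hero whoosh outro in Section 02, routed through the Fort_Events sound class.
# The cue is an empty shell with no root or sound nodes and a zero volume multiplier, so it produces no sound as authored.
# It still sets a very large max distance.
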SoundCue /PistolCanary_Event_Audio/Assets/Section02/Heroes/SC_PistolCanary_Whoosh_Outro
root=null
nodes []
properties {maxDistance=2097152, soundClassObject=&Fort_Events, volumeMultiplier=0}
